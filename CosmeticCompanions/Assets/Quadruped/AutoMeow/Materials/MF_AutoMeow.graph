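# Material function for the AutoMeow quadruped companion's visor, overriding base color and emissive on incoming material attributes and also outputting UVs.
# Visor UVs come from a selectable channel, scaled around center, offset and conditionally bump-offset for depth.
# A scanline function, opacity-blended, feeds a switchable emissive effect tinted by mask and eye colors.
# Named reroutes share the Fresnel term, flipbook opacity input and a visor mask texture's red channel.
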MaterialFunction /CosmeticCompanions/Assets/Quadruped/AutoMeow/Materials/MF_AutoMeow
graph {nodes [
  n1 = GetMaterialAttributes(None=FunctionInput() {inputName=Material, inputType=FunctionInput_MaterialAttributes}) {attributeGetTypes [#1, #2, #3]}
  n2 = Multiply(A=@n1.BaseColor, B=ScalarParameter() {group=Visor, parameterName=Visor_BasecolorScale, sortPriority=43})
  n3 = NamedRerouteUsage()
  FunctionOutput(None=SetMaterialAttributes!(
    MaterialAttributes=@n1.MaterialAttributes
    EmissiveColor=NamedRerouteUsage()
    BaseColor=LinearInterpolate(A=@n1.BaseColor, B=@n2, Alpha=@n3) {constB=0}
   ) {attributeSetTypes [#2, #3]})
  FunctionOutput(None=StaticSwitchParameter(True=NamedRerouteUsage(), False=TextureCoordinate()) {group=Visor, parameterName=Visor_OverrideBasicEmissiveUVs, sortPriority=21}) {outputName=UVs, sortPriority=1}
  n4 = LinearInterpolate(
   B=MaterialFunctionCall!(ScanLine_Tiling_1, Scanline_PannerSpeed_1, ScanLine_Tiling, Scanline_PannerSpeed, UVs=NamedRerouteUsage()) {materialFunction=/Game/Weapons/WeaponSkins/MaterialFunction/ScanLines}.Result
   Alpha=ScalarParameter() {defaultValue=0.5, group=Scanlines, parameterName=Scanline_Opacity, sortPriority=50}
  )
  Multiply(A=@n4, B=NamedRerouteUsage())
  NamedRerouteDeclaration(None=FunctionInput() {inputName=Flipbook Opacity, inputType=FunctionInput_Scalar, sortPriority=1}) {name=FlipbookOpacity, nodeColor=<rgba 0.1484 1 0 1>}
  NamedRerouteDeclaration(None=@n1.Normal) {name=Normal_IN, nodeColor=<rgba 1 0 0.7266 1>}
  n5 = Power(
   Base=OneMinus(None=Fresnel() {baseReflectFraction=0, exponent=1})
   Exp=ScalarParameter() {defaultValue=3, group=Visor, parameterName=Visor_FresnelPower, sortPriority=43}
  )
  NamedRerouteDeclaration(None=LinearInterpolate(B=Saturate(None=@n5), Alpha=ScalarParameter() {defaultValue=1, group=Visor, parameterName=Visor_FresnelIntensity, sortPriority=43}) {constA=1}) {name=Fresnel, nodeColor=<rgba 1 0.8203 0 1>}
  n6 = StaticSwitchParameter(True=TextureCoordinate() {coordinateIndex=1}, False=TextureCoordinate()) {group=Visor, parameterName=Visor_UseUV2, sortPriority=20}
  n7 = VectorParameter() {group=Visor, parameterName=Visor_UVScale, sortPriority=40, defaultValue=<rgba 1 1 0 1>}
  n8 = Add(
   A=MaterialFunctionCall(UVs=NamedRerouteDeclaration(None=@n6) {name=UV12, nodeColor=<rgba 0 0.0391 1 1>}, Texture Scale=ComponentMask(None=@n7.RGB) {g=true, r=true}) {materialFunction=&ScaleUVsByCenter}.UVs
   B=ComponentMask(None=VectorParameter() {group=Visor, parameterName=Visor_UVOffset, sortPriority=40, defaultValue=<rgba 0 0 0 1>}.RGB) {g=true, r=true}
  )
  n9 = ScalarParameter() {group=Visor, parameterName=Visor_Depth, sortPriority=41}
  n10 = Constant()
  n11 = NamedRerouteUsage()
  NamedRerouteDeclaration(None=If(
    A=NamedRerouteUsage()
    "A > B"=MaterialFunctionCall!(UV Coord=@n8, Height=@n9, ReferencePlane=@n10, HeightRatio, Normal=@n11) {materialFunction=/Game/Global/MaterialFunctions/MF_BumpOffset_Accurate}.Result
    "A < B"=NamedRerouteUsage()
   ) {constB=0.01}) {name=UVs, nodeColor=<rgba 0.875 1 0 1>}
  NamedRerouteDeclaration(None=@n1.EmissiveColor) {name=Emissive_IN, nodeColor=<rgba 1 0 0.9609 1>}
  n12 = Constant()
  n13 = VectorParameter() {group=Visor, parameterName=Visor_MaskTintColor, defaultValue=<rgba 1 1 1 1>}
  n14 = VectorParameter() {group=Visor, parameterName=EyeTintColor}
  n15 = Add(A=NamedRerouteUsage(), B=ScalarParameter() {group=Scanlines, parameterName=Scanline_ColorBias, sortPriority=41})
  n16 = Power(Base=@n15)
  n17 = NamedRerouteUsage()
  n18 = Multiply(
   A=LinearInterpolate(B=LinearInterpolate(A=@n13.RGB, B=@n14.RGB, Alpha=@n16), Alpha=Multiply(A=@n15, B=@n17))
   B=ScalarParameter() {defaultValue=1, group=Visor, parameterName=Visor_BaseBrightness, sortPriority=41}
  )
  n19 = NamedRerouteUsage()
  NamedRerouteDeclaration(None=StaticSwitchParameter(True=LinearInterpolate(A=@n12, B=@n18, Alpha=@n19) {constB=0}, False=NamedRerouteUsage()) {group=Visor Add-Ons, parameterName=Visor_UseScanline, sortPriority=20}) {name=Emissive_OUT, nodeColor=<rgba 0 1 0.7188 1>}
  NamedRerouteDeclaration() {name=BasicVisor, nodeColor=<rgba 0.8984 1 0 1>}
  NamedRerouteDeclaration(None=@n4) {name=Scanline, nodeColor=<rgba 1 0.2109 0 1>}
  n20 = NamedRerouteDeclaration(None=TextureSampleParameter2D() {group=Scanlines, parameterName=VisorMask}.R) {name=VisorMask, nodeColor=<rgba 1 0 0.5156 1>}
  NamedRerouteDeclaration(None=@n20) {name=TVStaticMask, nodeColor=<rgba 0.7109 1 0 1>}
  NamedRerouteDeclaration(None=@n20) {name=ScanlinesMask, nodeColor=<rgba 1 0.5116 0.5981 1>}
  NamedRerouteDeclaration(None=@n20) {name=SweepDistortionMask, nodeColor=<rgba 0.161 1 0.707 1>}
  NamedRerouteDeclaration() {name=Debug, nodeColor=<rgba 0 1 0.5078 1>}
  NamedRerouteUsage()
  NamedRerouteUsage()
  NamedRerouteUsage()
 ]}
