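# Eye material for the AutoMeow quadruped companion, marked deprecated, that draws animation-driven expressions from a 3x3 flipbook.
# Pose-curve parameters become weighted indices combined through a Max chain to select expressions such as angry, closed or happy.
# The flipbook's HSV value blends two tints over base color from character parent, fur and AutoMeow functions, before PostFX.
# On the emote frame, an SDF box sized by eye_emote_02 masks the overlay.
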Material /CosmeticCompanions/Assets/Quadruped/AutoMeow/Materials/M_AutoMeow_Eyes_Deprecated
graph {
 expressions {
  MaterialExpressionAdd_0 {
   class=MaterialExpressionAdd
   inputs [
    A {from=MaterialExpressionTextureCoordinate_0}
    B {from=MaterialExpressionReroute_8}
   ]
  }
  MaterialExpressionAdd_5 {
   class=MaterialExpressionAdd
   inputs [
    A {from=MaterialExpressionAdd_7}
    B {from=MaterialExpressionMultiply_2}
   ]
  }
  MaterialExpressionAdd_7 {
   class=MaterialExpressionAdd
   inputs [
    A {from=MaterialExpressionMultiply_5}
    B {from=MaterialExpressionConstant2Vector_6}
   ]
  }
  MaterialExpressionAppendVector_2 {
   class=MaterialExpressionAppendVector
   inputs [
    A {from=MaterialExpressionMultiply_4, output=""}
    B {from=MaterialExpressionConstant_3, output=""}
   ]
  }
  MaterialExpressionComponentMask_0 {
   class=MaterialExpressionComponentMask
   inputs [None {from=MaterialExpressionMaterialFunctionCall_10, output="RGB to HSV (0-1)"}]
   properties {b=true}
  }
  MaterialExpressionConstant2Vector_10 {class=MaterialExpressionConstant2Vector, properties {g=1, r=1}}
  MaterialExpressionConstant2Vector_4 {class=MaterialExpressionConstant2Vector, properties {g=3, r=3}}
  MaterialExpressionConstant2Vector_6 {class=MaterialExpressionConstant2Vector, properties {r=-0.4705}}
  MaterialExpressionConstant2Vector_7 {class=MaterialExpressionConstant2Vector, properties {g=-0.13, r=-0.14}}
  MaterialExpressionConstant2Vector_8 {class=MaterialExpressionConstant2Vector, properties {g=0.01, r=0.22}}
  MaterialExpressionConstant_10 {class=MaterialExpressionConstant, properties {r=4}}
  MaterialExpressionConstant_11 {class=MaterialExpressionConstant, properties {r=5}}
  MaterialExpressionConstant_12 {class=MaterialExpressionConstant, properties {r=6}}
  MaterialExpressionConstant_13 {class=MaterialExpressionConstant, properties {r=7}}
  MaterialExpressionConstant_2 {class=MaterialExpressionConstant, properties {r=0.066}}
  MaterialExpressionConstant_3 {class=MaterialExpressionConstant, properties {r=0.032}}
  MaterialExpressionConstant_4 {class=MaterialExpressionConstant, properties {r=1}}
  MaterialExpressionConstant_6 {class=MaterialExpressionConstant}
  MaterialExpressionConstant_7 {class=MaterialExpressionConstant, properties {r=2}}
  MaterialExpressionConstant_9 {class=MaterialExpressionConstant, properties {r=3}}
  MaterialExpressionConvert_0 {
   class=MaterialExpressionConvert
   inputs [
    None {from=MaterialExpressionDivide_4}
    None {from=""}
   ]
   properties {
    nodeName=Make Vector2
    convertInputs [
     {
      type=Scalar
      expressionInput {
       expression=&self:MaterialExpressionDivide_4
       inputName=None
       mask=0
       maskA=0
       maskB=0
       maskG=0
       maskR=0
       outputIndex=0
      }
     }
     {type=Scalar, expressionInput {expression=null, inputName=None, mask=0, maskA=0, maskB=0, maskG=0, maskR=0, outputIndex=0}}
    ]
    convertOutputs [{type=Vector2}]
   }
  }
  MaterialExpressionDivide_4 {
   class=MaterialExpressionDivide
   inputs [
    A {from=MaterialExpressionOneMinus_6}
    B {from=""}
   ]
   properties {constB=-2}
  }
  MaterialExpressionGetMaterialAttributes_0 {
   class=MaterialExpressionGetMaterialAttributes
   inputs [None {from=MaterialExpressionReroute_4, output=""}]
   properties {attributeGetTypes [#1]}
  }
  MaterialExpressionIf_0 {
   class=MaterialExpressionIf
   inputs [
    A {from=MaterialExpressionReroute_9, output=""}
    B {from="", output=""}
    "A > B" {from=MaterialExpressionConstant_4, output=""}
    "A == B" {from=MaterialExpressionStep_3, output=""}
    "A < B" {from=MaterialExpressionConstant_4, output=""}
   ]
   properties {constB=3}
  }
  MaterialExpressionLinearInterpolate_0 {
   class=MaterialExpressionLinearInterpolate
   inputs [
    A {from=MaterialExpressionGetMaterialAttributes_0, output=Base Color}
    B {from=MaterialExpressionReroute_6}
    Alpha {from=MaterialExpressionReroute_7}
   ]
  }
  MaterialExpressionLinearInterpolate_3 {
   class=MaterialExpressionLinearInterpolate
   inputs [
    A {from=MaterialExpressionVectorParameter_0, output=RGB}
    B {from=MaterialExpressionVectorParameter_1, output=RGB}
    Alpha {from=MaterialExpressionComponentMask_0}
   ]
  }
  MaterialExpressionMaterialFunctionCall_0 {
   class=MaterialExpressionMaterialFunctionCall
   inputs [
    Overide Texture Set UVs {from=""}
    Override Tech Art Mask UVs {from=""}
    Override Basic Emissive UVs {from=MaterialExpressionReroute_3}
   ]
   properties {materialFunction=&MF_CharacterParent_2023_PreFX}
  }
  MaterialExpressionMaterialFunctionCall_1 {
   class=MaterialExpressionMaterialFunctionCall
   inputs [
    Material {from=MaterialExpressionSetMaterialAttributes_0}
    M {from=MaterialExpressionMaterialFunctionCall_0, output=M}
    Pre Skinned Local Position {from=MaterialExpressionMaterialFunctionCall_0, output=M}
    Distance Blend {from=MaterialExpressionMaterialFunctionCall_0, output=M}
    Fresnel {from=MaterialExpressionMaterialFunctionCall_0, output=M}
   ]
   properties {materialFunction=&MF_CharacterParent_2023_PostFX}
  }
  MaterialExpressionMaterialFunctionCall_10 {
   class=MaterialExpressionMaterialFunctionCall
   inputs [RGB {from=MaterialExpressionMaterialFunctionCall_9, output=Color}]
   properties {materialFunction=/Engine/Functions/Engine_MaterialFunctions02/Math/RGBtoHSV}
  }
  MaterialExpressionMaterialFunctionCall_2 {
   class=MaterialExpressionMaterialFunctionCall
   inputs [
    Material {from=""}
    Tech Art Mask {from=""}
   ]
   properties {materialFunction=/CosmeticCompanions/Materials/MaterialFunctions/MF_FN_Companion_Custom_Color}
  }
  MaterialExpressionMaterialFunctionCall_3 {
   class=MaterialExpressionMaterialFunctionCall
   inputs [
    Material {from=MaterialExpressionMaterialFunctionCall_7, output=Result}
    Flipbook Opacity {from=""}
   ]
   properties {materialFunction=&/MF_AutoMeow}
  }
  MaterialExpressionMaterialFunctionCall_4 {
   class=MaterialExpressionMaterialFunctionCall
   inputs [
    UVs {from=MaterialExpressionAdd_5}
    Size {from=MaterialExpressionAppendVector_2}
    CornerRoundness {from=""}
    KeepAspectRatio {from=""}
    PixelSize {from=""}
   ]
   properties {materialFunction=/FortUILibrary/MaterialFunctions/Shapes/SDFs/MF_SDF_Box}
  }
  MaterialExpressionMaterialFunctionCall_7 {
   class=MaterialExpressionMaterialFunctionCall
   inputs [
    Material {from=MaterialExpressionMaterialFunctionCall_0, output=Result}
    TechArtMask {from=MaterialExpressionMaterialFunctionCall_0, output=Result}
   ]
   properties {materialFunction=/Game/Global/Materials/Character_Parent_2023/Fur_Parent_2023/MF_Fur_CustomColor}
  }
  MaterialExpressionMaterialFunctionCall_9 {
   class=MaterialExpressionMaterialFunctionCall
   inputs [
    Flipbook Index {from=MaterialExpressionReroute_5}
    Flipbook Texture {from=MaterialExpressionTextureObject_0}
    Flipbook Column Count {from=MaterialExpressionConstant2Vector_4, output=G}
    Flipbook Row Count {from=MaterialExpressionConstant2Vector_4, output=G}
    UVs {from=MaterialExpressionTextureCoordinate_0}
    UV Offset {from=MaterialExpressionConstant2Vector_8}
    UV Scale {from=MaterialExpressionConstant2Vector_7}
    MirrorLeftUV {from=""}
   ]
   properties {materialFunction=/Game/Global/Materials/Character_Parent_2023/FlipbookFace_Parent_2023/MaterialFunctions/Common/MF_FlipbookTexture}
  }
  MaterialExpressionMax_0 {
   class=MaterialExpressionMax
   inputs [
    A {from=MaterialExpressionConstant_6}
    B {from=MaterialExpressionMin_0}
   ]
   properties {desc=Base}
  }
  MaterialExpressionMax_1 {
   class=MaterialExpressionMax
   inputs [
    A {from=MaterialExpressionMax_0}
    B {from=MaterialExpressionMultiply_7}
   ]
  }
  MaterialExpressionMax_2 {
   class=MaterialExpressionMax
   inputs [
    A {from=MaterialExpressionMax_1}
    B {from=MaterialExpressionMultiply_8}
   ]
  }
  MaterialExpressionMax_3 {
   class=MaterialExpressionMax
   inputs [
    A {from=MaterialExpressionMax_2}
    B {from=MaterialExpressionMultiply_9}
   ]
  }
  MaterialExpressionMax_4 {
   class=MaterialExpressionMax
   inputs [
    A {from=MaterialExpressionMax_3}
    B {from=MaterialExpressionMultiply_10}
   ]
  }
  MaterialExpressionMax_5 {
   class=MaterialExpressionMax
   inputs [
    A {from=MaterialExpressionMax_4}
    B {from=MaterialExpressionMultiply_11}
   ]
  }
  MaterialExpressionMax_6 {
   class=MaterialExpressionMax
   inputs [
    A {from=MaterialExpressionMax_5}
    B {from=MaterialExpressionMultiply_12}
   ]
  }
  MaterialExpressionMin_0 {
   class=MaterialExpressionMin
   inputs [
    A {from=MaterialExpressionScalarParameter_38, output=""}
    B {from=MaterialExpressionScalarParameter_4, output=""}
   ]
   properties {desc=Angry}
  }
  MaterialExpressionMin_1 {
   class=MaterialExpressionMin
   inputs [
    A {from=MaterialExpressionScalarParameter_0, output=""}
    B {from=MaterialExpressionScalarParameter_3, output=""}
   ]
   properties {}
  }
  MaterialExpressionMin_2 {
   class=MaterialExpressionMin
   inputs [
    A {from=MaterialExpressionScalarParameter_38, output=""}
    B {from=MaterialExpressionScalarParameter_36, output=""}
   ]
   properties {}
  }
  MaterialExpressionMin_3 {
   class=MaterialExpressionMin
   inputs [
    A {from=MaterialExpressionMin_1, output=""}
    B {from=MaterialExpressionScalarParameter_5, output=""}
   ]
   properties {}
  }
  MaterialExpressionMin_4 {
   class=MaterialExpressionMin
   inputs [
    A {from=MaterialExpressionScalarParameter_36, output=""}
    B {from=MaterialExpressionScalarParameter_7, output=""}
   ]
   properties {}
  }
  MaterialExpressionMultiply_10 {
   class=MaterialExpressionMultiply
   inputs [
    A {from=MaterialExpressionConstant_11}
    B {from=MaterialExpressionMin_3}
   ]
   properties {desc=Eyes Closed}
  }
  MaterialExpressionMultiply_11 {
   class=MaterialExpressionMultiply
   inputs [
    A {from=MaterialExpressionConstant_12}
    B {from=MaterialExpressionMin_4}
   ]
   properties {desc=Worried}
  }
  MaterialExpressionMultiply_12 {
   class=MaterialExpressionMultiply
   inputs [
    A {from=MaterialExpressionConstant_13}
    B {from=MaterialExpressionStep_2}
   ]
   properties {desc=Happy}
  }
  MaterialExpressionMultiply_2 {
   class=MaterialExpressionMultiply
   inputs [
    A {from=MaterialExpressionConvert_0}
    B {from=""}
   ]
   properties {constB=-1}
  }
  MaterialExpressionMultiply_4 {
   class=MaterialExpressionMultiply
   inputs [
    A {from=MaterialExpressionScalarParameter_2}
    B {from=MaterialExpressionConstant_2}
   ]
  }
  MaterialExpressionMultiply_5 {
   class=MaterialExpressionMultiply
   inputs [
    A {from=MaterialExpressionAdd_0}
    B {from=MaterialExpressionConstant2Vector_10}
   ]
   properties {constB=-1}
  }
  MaterialExpressionMultiply_6 {
   class=MaterialExpressionMultiply
   inputs [
    A {from=MaterialExpressionMaterialFunctionCall_9, output=Alpha}
    B {from=MaterialExpressionIf_0}
   ]
  }
  MaterialExpressionMultiply_7 {
   class=MaterialExpressionMultiply
   inputs [
    A {from=MaterialExpressionConstant_7}
    B {from=MaterialExpressionStep_0}
   ]
   properties {desc=Wide}
  }
  MaterialExpressionMultiply_8 {
   class=MaterialExpressionMultiply
   inputs [
    A {from=MaterialExpressionConstant_9}
    B {from=MaterialExpressionStep_1}
   ]
   properties {desc=Emote}
  }
  MaterialExpressionMultiply_9 {
   class=MaterialExpressionMultiply
   inputs [
    A {from=MaterialExpressionConstant_10}
    B {from=MaterialExpressionMin_2}
   ]
   properties {desc=Eyes Down}
  }
  MaterialExpressionOneMinus_6 {class=MaterialExpressionOneMinus, inputs [None {from=MaterialExpressionMultiply_4}]}
  MaterialExpressionReroute_0 {class=MaterialExpressionReroute, inputs [None {from=MaterialExpressionReroute_2}]}
  MaterialExpressionReroute_1 {class=MaterialExpressionReroute, inputs [None {from=MaterialExpressionReroute_0}]}
  MaterialExpressionReroute_2 {class=MaterialExpressionReroute, inputs [None {from=MaterialExpressionMaterialFunctionCall_3, output=UVs}]}
  MaterialExpressionReroute_3 {class=MaterialExpressionReroute, inputs [None {from=MaterialExpressionReroute_1}]}
  MaterialExpressionReroute_4 {class=MaterialExpressionReroute, inputs [None {from=MaterialExpressionMaterialFunctionCall_3, output=Result}]}
  MaterialExpressionReroute_5 {class=MaterialExpressionReroute, inputs [None {from=MaterialExpressionMax_6}]}
  MaterialExpressionReroute_6 {class=MaterialExpressionReroute, inputs [None {from=MaterialExpressionLinearInterpolate_3}]}
  MaterialExpressionReroute_7 {class=MaterialExpressionReroute, inputs [None {from=MaterialExpressionMultiply_6}]}
  MaterialExpressionReroute_8 {class=MaterialExpressionReroute, inputs [None {from=MaterialExpressionConstant2Vector_8}]}
  MaterialExpressionReroute_9 {class=MaterialExpressionReroute, inputs [None {from=MaterialExpressionReroute_5}]}
  MaterialExpressionScalarParameter_0 {class=MaterialExpressionScalarParameter, properties {group="Eyes Curve Parameters (Driven By Animation)", parameterName=R_blink_pose}}
  MaterialExpressionScalarParameter_1 {class=MaterialExpressionScalarParameter, properties {group="Eyes Curve Parameters (Driven By Animation)", parameterName=eye_emote_01}}
  MaterialExpressionScalarParameter_2 {class=MaterialExpressionScalarParameter, properties {group="Eyes Curve Parameters (Driven By Animation)", parameterName=eye_emote_02}}
  MaterialExpressionScalarParameter_3 {class=MaterialExpressionScalarParameter, properties {group="Eyes Curve Parameters (Driven By Animation)", parameterName=R_upperLidDown_pose}}
  MaterialExpressionScalarParameter_36 {class=MaterialExpressionScalarParameter, properties {group="Eyes Curve Parameters (Driven By Animation)", parameterName=R_brow_Ot_Dn_pose}}
  MaterialExpressionScalarParameter_38 {class=MaterialExpressionScalarParameter, properties {group="Eyes Curve Parameters (Driven By Animation)", parameterName=R_brow_In_Dn_pose}}
  MaterialExpressionScalarParameter_4 {
   class=MaterialExpressionScalarParameter
   properties {group="Eyes Curve Parameters (Driven By Animation)", parameterName=R_brow_sqeeze_in_pose}
  }
  MaterialExpressionScalarParameter_41 {class=MaterialExpressionScalarParameter, properties {group="Eyes Curve Parameters (Driven By Animation)", parameterName=R_wide_pose}}
  MaterialExpressionScalarParameter_5 {class=MaterialExpressionScalarParameter, properties {group="Eyes Curve Parameters (Driven By Animation)", parameterName=R_lower_lid_up_pose}}
  MaterialExpressionScalarParameter_7 {
   class=MaterialExpressionScalarParameter
   properties {group="Eyes Curve Parameters (Driven By Animation)", parameterName=R_brow_squeeze_in_pose}
  }
  MaterialExpressionScalarParameter_8 {class=MaterialExpressionScalarParameter, properties {group="Eyes Curve Parameters (Driven By Animation)", parameterName=R_smile_pose}}
  MaterialExpressionSetMaterialAttributes_0 {
   class=MaterialExpressionSetMaterialAttributes
   inputs [
    MaterialAttributes {from=MaterialExpressionGetMaterialAttributes_0, output=MaterialAttributes}
    Base Color {from=MaterialExpressionLinearInterpolate_0, output=""}
   ]
   properties {attributeSetTypes [#1]}
  }
  MaterialExpressionStep_0 {
   class=MaterialExpressionStep
   inputs [
    Y {from="", output=""}
    X {from=MaterialExpressionScalarParameter_41, output=""}
   ]
   properties {constX=2, constY=1}
  }
  MaterialExpressionStep_1 {
   class=MaterialExpressionStep
   inputs [
    Y {from="", output=""}
    X {from=MaterialExpressionScalarParameter_1, output=""}
   ]
   properties {constX=2, constY=1}
  }
  MaterialExpressionStep_2 {
   class=MaterialExpressionStep
   inputs [
    Y {from="", output=""}
    X {from=MaterialExpressionScalarParameter_8, output=""}
   ]
   properties {constX=2, constY=1}
  }
  MaterialExpressionStep_3 {
   class=MaterialExpressionStep
   inputs [
    Y {from="", output=""}
    X {from=MaterialExpressionMaterialFunctionCall_4, output=SDF}
   ]
   properties {}
  }
  MaterialExpressionTextureCoordinate_0 {class=MaterialExpressionTextureCoordinate}
  MaterialExpressionTextureObject_0 {
   class=MaterialExpressionTextureObject
   inputs []
   properties {texture=/CosmeticCompanions/Assets/Quadruped/AutoMeow/Textures/T_AutoMeow_Eyes_FlipbookFace}
  }
  MaterialExpressionVectorParameter_0 {
   class=MaterialExpressionVectorParameter
   properties {group=Color Customization, parameterName=EyeTintColor, defaultValue=<rgba 0.3095 0.9216 0.9131 1>}
  }
  MaterialExpressionVectorParameter_1 {
   class=MaterialExpressionVectorParameter
   properties {group=Color Customization, parameterName=EyeTintColor2, defaultValue=<rgba 0.6514 1 0.956 1>}
  }
 }
 properties {
  bAutomaticallySetUsageInEditor=false
  bUseMaterialAttributes=true
  bUsedWithMeshDeformer=true
  bUsedWithMorphTargets=true
  bUsedWithSkeletalMesh=true
  bUsedWithStaticLighting=true
  previewMesh=/CosmeticCompanions/Assets/Quadruped/AutoMeow/Meshes/AutoMeow_Quadruped
  shadingModel=MSM_FromMaterialExpression
  usedShadingModels=Default Lit | Subsurface | Clear Coat | Subsurface Profile | Cloth
 }
}
