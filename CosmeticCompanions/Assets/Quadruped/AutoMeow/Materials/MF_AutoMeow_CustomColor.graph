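# Material function that adds customizable colors to the AutoMeow quadruped companion's material attributes, toggled by the UseCustomColors static switch.
# Channels of a mask texture define primary and two accent regions, with base color and its desaturated detail available for layered tint blends.
# A desaturated, smoothstepped mask drives a visor tint that overlays base color and replaces emissive.
# An EyeTintColor parameter is present but unconnected.
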MaterialFunction /CosmeticCompanions/Assets/Quadruped/AutoMeow/Materials/MF_AutoMeow_CustomColor
graph {nodes [
  n1 = NamedRerouteUsage()
  n2 = LinearInterpolate(
   A=NamedRerouteUsage()
   B=Multiply(
    A=VectorParameter() {group=CustomizableColor, parameterName=PrimaryColor, sortPriority=12, defaultValue=<rgba 1 1 1 1>}.RGB
    B=NamedRerouteUsage()
   )
   Alpha=NamedRerouteUsage()
  ) {constA=1}
  n3 = VectorParameter() {group=CustomizableColor, parameterName=AccentColor_01, sortPriority=14, defaultValue=<rgba 1 1 1 1>}
  n4 = Multiply(A=@n3.RGB, B=NamedRerouteUsage())
  n5 = NamedRerouteUsage()
  n6 = VectorParameter() {group=CustomizableColor, parameterName=AccentColor_02, sortPriority=16, defaultValue=<rgba 1 1 1 1>}
  n7 = NamedRerouteUsage()
  n8 = SmoothStep(Value=Desaturation(None=NamedRerouteUsage())) {constMax=0.1}
  n9 = Multiply(A=VectorParameter() {parameterName=Visor_MaskTintColor}.RGB, B=@n8)
  n10 = LinearInterpolate(A=LinearInterpolate(A=LinearInterpolate(A=@n2, B=@n4, Alpha=@n5), B=Multiply(A=@n6.RGB, B=@n7), Alpha=NamedRerouteUsage()), B=@n9, Alpha=@n8) {constAlpha=0, constB=0}
  FunctionOutput(None=StaticSwitchParameter(
    True=SetMaterialAttributes!(MaterialAttributes=@n1, BaseColor=@n10, EmissiveColor=@n9) {attributeSetTypes [#1, #2]}
    False=NamedRerouteUsage()
   ) {group=CustomizableColor, parameterName=UseCustomColors, sortPriority=-1})
  n11 = GetMaterialAttributes(None=NamedRerouteDeclaration(None=FunctionInput() {inputName=Material, inputType=FunctionInput_MaterialAttributes}) {name=IN_Material, nodeColor=<rgba 0.6172 0 1 1>}) {attributeGetTypes [#1, #2]}
  NamedRerouteDeclaration(None=@n11.EmissiveColor) {name=In_EmissiveColor, nodeColor=<rgba 1 0 0 1>}
  n12 = NamedRerouteDeclaration(None=@n11.BaseColor) {name=IN_BaseColor, nodeColor=<rgba 1 0 0.0234 1>}
  NamedRerouteDeclaration(None=SmoothStep(Value=Desaturation(None=@n12)) {constMax=0.2}) {name=BaseColorDetail, nodeColor=<rgba 0.4766 0 1 1>}
  n13 = TextureSampleParameter2D() {
   group=CustomizableColor
   parameterName=Custom Color Mask
   samplerSource=SSM_Wrap_WorldGroupSettings
   samplerType=SAMPLERTYPE_Masks
   sortPriority=0
   texture=&T_BlackMask
  }
  NamedRerouteDeclaration(None=SmoothStep(Value=ChannelMaskParameter(None=@n13.RGB) {group=CustomizableColor, parameterName=PrimaryColor_Channel, sortPriority=11}.R) {constMax=0.3, constMin=0.01}) {name=PrimaryColor_Mask, nodeColor=<rgba 1 0.0596 0.1687 1>}
  n14 = ChannelMaskParameter(None=@n13.RGB) {group=CustomizableColor, maskChannel=Green, parameterName=AccentColor_01_Channel, sortPriority=13, defaultValue=<rgba 0 1 0 0>}
  NamedRerouteDeclaration(None=@n14.R) {name=AccentColor_01_Mask, nodeColor=<rgba 0.4929 1 0.2743 1>}
  n15 = ChannelMaskParameter(None=@n13.RGB) {group=CustomizableColor, maskChannel=Blue, parameterName=AccentColor_02_Channel, sortPriority=15, defaultValue=<rgba 0 0 1 0>}
  NamedRerouteDeclaration(None=@n15.R) {name=AccentColor_02_Mask, nodeColor=<rgba 0 0.2243 1 1>}
  VectorParameter() {parameterName=EyeTintColor}
 ]}
