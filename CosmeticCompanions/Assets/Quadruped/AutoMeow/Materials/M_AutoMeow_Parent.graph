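# M_AutoMeow_Parent is the parent material for the AutoMeow quadruped cosmetic companion.
# It uses material attributes, passing the shared 2023 character PreFX function through an AutoMeow custom color function into the matching PostFX function.
# PreFX also feeds auxiliary PostFX inputs directly, bypassing the color step.
# The shading model comes from the material expression, with usage enabled for skeletal meshes, morph targets and mesh deformers.
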Material /CosmeticCompanions/Assets/Quadruped/AutoMeow/Materials/M_AutoMeow_Parent
graph {
 properties {
  bAutomaticallySetUsageInEditor=false
  bUseMaterialAttributes=true
  bUsedWithMeshDeformer=true
  bUsedWithMorphTargets=true
  bUsedWithSkeletalMesh=true
  bUsedWithStaticLighting=true
  previewMesh=/CosmeticCompanions/Assets/Quadruped/AutoMeow/Meshes/AutoMeow_Quadruped
  shadingModel=MSM_FromMaterialExpression
  usedShadingModels=Default Lit | Subsurface | Clear Coat | Subsurface Profile | Cloth
 }
 nodes [
  n1 = MaterialFunctionCall() {materialFunction=&MF_CharacterParent_2023_PreFX}
  MaterialFunctionCall(
   Material=MaterialFunctionCall!(Material=@n1.Result) {materialFunction=&/MF_AutoMeow_CustomColor}.Result
   M=@n1.M
   Pre Skinned Local Position=@n1.M
   Distance Blend=@n1.M
   Fresnel=@n1.M
  ) {materialFunction=&MF_CharacterParent_2023_PostFX}
 ]
}
