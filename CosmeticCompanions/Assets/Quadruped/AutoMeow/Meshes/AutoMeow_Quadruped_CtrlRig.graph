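# Control Rig for the AutoMeow quadruped companion that constrains the left and right extra collar bones against a capsule on the head bone.
# The forward solve calls the reusable Clamp Capsule function once per collar bone, with debug drawing enabled.
# The function offsets the bone transform, spatially clamps it in the offset capsule bone's space, removes the offset and writes the result back weighted by Alpha.
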ControlRigBlueprint /CosmeticCompanions/Assets/Quadruped/AutoMeow/Meshes/AutoMeow_Quadruped_CtrlRig
graphs [
 New Function_ContainedGraph {nodes [
   n1 = RigVMFunctionEntryNode()
   Multiply = &RigVMFunction_MathTransformMul(A=@n1.CapsuleOffset, B=&RigUnit_GetTransform(bInitial=False, Item=@n1.CapsuleBone).Transform) {template="Multiply::Execute(in A,in B,out Result)"}
   RigUnit_GetTransform = &RigUnit_GetTransform(bInitial=False, Item=@n1.MoveBone)
   ClampSpatially = /Script/RigVM.RigVMFunction_MathTransformClampSpatially(
    DebugColor=(R=1,G=0,B=0,A=1)
    DebugThickness=1
    Type=Capsule
    Axis=@n1.CapsuleAxis
    Maximum=@n1.CapsuleLength
    Minimum=@n1.CapsuleWidth
    bDrawDebug=@n1.DrawDegub
    Space=@Multiply.Result
    Value=&RigVMFunction_MathTransformMul(B=@n1.BoneOffset, A=@RigUnit_GetTransform.Transform) {template="Multiply::Execute(in A,in B,out Result)"}.Result
   ) {
    is=struct RigVMUnitNode
    template="ClampSpatially::Execute(in Value,in Axis,in Type,in Minimum,in Maximum,in Space,in bDrawDebug,in DebugColor,in DebugThickness,out Result)"
   }
   Inverse = /Script/RigVM.RigVMFunction_MathTransformInverse(Value=@n1.BoneOffset) {is=struct RigVMUnitNode, template="Inverse::Execute(in Value,out Result)"}
   > RigUnit_SetTransform = &RigUnit_SetTransform(
    bInitial=False
    Weight=@n1.Alpha
    Value=&RigVMFunction_MathTransformMul(A=@ClampSpatially.Result, B=@Inverse.Result) {template="Multiply::Execute(in A,in B,out Result)"}.Result
    Item=@n1.MoveBone
   )
   > RigVMFunctionReturnNode()
   &RigVMFunction_MathTransformMul(B=(Rotation=(X=0,Y=0,Z=0,W=1),Translation=(X=-1,Y=-1,Z=-1),Scale3D=(X=1,Y=1,Z=1)), A=@n1.BoneOffset) {template="Multiply::Execute(in A,in B,out Result)"}
  ]}
 RigVMFunctionLibrary {nodes [Clamp Capsule = RigVMCollapseNode(
    Alpha=20
    BoneOffset=(Rotation=(X=0,Y=0,Z=0,W=1),Translation=(X=0,Y=0,Z=0),Scale3D=(X=1,Y=1,Z=1))
    CapsuleAxis=Z
    CapsuleBone=(Type=None,Name="None")
    CapsuleLength=20
    CapsuleOffset=(Rotation=(X=0,Y=0,Z=0,W=1),Translation=(X=0,Y=0,Z=0),Scale3D=(X=1,Y=1,Z=1))
    CapsuleWidth=20
    ExecuteContext=()
    MoveBone=(Type=None,Name="None")
   )]}
 RigVMModel {nodes [
   RigUnit_BeginExecution = &RigUnit_BeginExecution()
   Clamp Capsule_1 = RigVMFunctionReferenceNode(
    Alpha=20
    BoneOffset=(Rotation=(X=0,Y=0,Z=0,W=1),Translation=(X=0,Y=0,Z=0),Scale3D=(X=1,Y=1,Z=1))
    CapsuleAxis=X
    CapsuleBone=(Type=Bone,Name="head")
    CapsuleLength=30
    CapsuleOffset=(Rotation=(X=0,Y=0,Z=0,W=1),Translation=(X=0,Y=0,Z=0),Scale3D=(X=1,Y=1,Z=1))
    CapsuleWidth=10
    DrawDegub=True
    MoveBone=(Type=Bone,Name="extra_collar_l")
    ExecuteContext=@RigUnit_BeginExecution.ExecutePin
   )
   > Clamp Capsule_2 = RigVMFunctionReferenceNode(
    Alpha=20
    BoneOffset=(Rotation=(X=0,Y=0,Z=0,W=1),Translation=(X=0,Y=0,Z=0),Scale3D=(X=1,Y=1,Z=1))
    CapsuleAxis=X
    CapsuleBone=(Type=Bone,Name="head")
    CapsuleLength=30
    CapsuleOffset=(Rotation=(X=0,Y=0,Z=0,W=1),Translation=(X=0,Y=0,Z=0),Scale3D=(X=1,Y=1,Z=1))
    CapsuleWidth=10
    DrawDegub=True
    MoveBone=(Type=Bone,Name="extra_collar_r")
   )
  ]}
]
hierarchy [
 root {loc=<0 0 0>, rot=<0 0 0 1>}
 pelvis {parent=root, loc=<0 -14.664 26.7066>, rot=<0.5008 0.4992 -0.5008 -0.4992>}
 spine_01 {parent=pelvis, loc=<0 0 0>, rot=<0 0 -0.0124 -0.9999>}
 spine_02 {parent=spine_01, loc=<14.2907 -0.3558 0>, rot=<0 0 0.0159 -0.9999>}
 spine_03 {parent=spine_02, loc=<13.1068 0 0>, rot=<0 0 0.0041 -1>}
 neck_01 {parent=spine_03, loc=<2.5947 -1.5005 0>, rot=<0 0 0.4032 -0.9151>}
 neck_02 {parent=neck_01, loc=<5.711 0 0>, rot=<0 0 0.0462 -0.9989>}
 head {parent=neck_02, loc=<6.5839 0 0>, rot=<-0.2129 -0.6743 0.2129 -0.6743>}
 face_attach {parent=head, loc=<0 -0.8148 -0.4208>, rot=<0 0 0 1>}
 jaw {parent=face_attach, loc=<0 1.0725 1.4273>, rot=<0 0.7071 0 -0.7071>}
 cheek_bt_l {parent=face_attach, loc=<6.5575 5.747 -1.6239>, rot=<0.0809 -0.1726 -0.1547 -0.9694>}
 cheek_bt_r {parent=face_attach, loc=<-6.5575 5.7469 -1.6239>, rot=<0.9694 -0.1547 0.1726 0.0809>}
 reverse_jaw {parent=face_attach, loc=<0 0.0433 0.7402>, rot=<0 0.7071 0 -0.7071>}
 ear_01_l {parent=reverse_jaw, loc=<1.5571 0.0575 -0.5535>, rot=<-0.0367 -0.3547 0.0475 -0.9331>}
 ear_02_l {parent=ear_01_l, loc=<10.7734 0 0>, rot=<0.2138 -0.0589 -0.1314 -0.9662>}
 ear_03_l {parent=ear_02_l, loc=<2.3632 0 0>, rot=<0.0137 0.0739 -0.0375 -0.9965>}
 ear_04_l {parent=ear_03_l, loc=<2.608 0 0>, rot=<-0.0159 -0.0128 -0.0425 -0.9989>}
 ear_01_r {parent=reverse_jaw, loc=<1.5571 0.0575 0.5535>, rot=<-0.3547 0.0367 0.9331 0.0475>}
 ear_02_r {parent=ear_01_r, loc=<-10.7734 0 0>, rot=<0.2138 -0.0589 -0.1314 -0.9662>}
 ear_03_r {parent=ear_02_r, loc=<-2.3632 -0.0001 0>, rot=<0.0137 0.0739 -0.0375 -0.9965>}
 ear_04_r {parent=ear_03_r, loc=<-2.6079 0.0001 -0.0001>, rot=<-0.0159 -0.0128 -0.0425 -0.9989>}
 extra_collar_l {parent=neck_01, loc=<4.2551 0.1468 -7.4824>, rot=<0.4036 -0.5018 -0.5615 -0.5196>}
 extra_collar_r {parent=neck_01, loc=<4.2551 0.1468 7.4824>, rot=<-0.5106 -0.3954 0.5259 -0.5536>}
 shoulder_fr_l {parent=spine_03, loc=<1.9267 -0.6483 -5.7705>, rot=<0.0135 -0.0015 -0.7949 -0.6066>}
 elbow_fr_l {parent=shoulder_fr_l, loc=<10.8758 0 0>, rot=<-0.0045 0.004 0.1185 -0.9929>}
 wrist_fr_l {parent=elbow_fr_l, loc=<11.4815 0 0>, rot=<0.9838 -0.1787 0.0143 -0.005>}
 ball_fr_l {parent=wrist_fr_l, loc=<3.7813 0 0>, rot=<-0.0436 0.0565 -0.4345 -0.8978>}
 toe_fr_l {parent=ball_fr_l, loc=<1.4234 0 0>, rot=<-0.0136 -0.0343 0.0437 -0.9984>}
 toe_tip_fr_l {parent=toe_fr_l, loc=<3.1113 0 0>, rot=<0.0031 0.006 -0.1897 -0.9818>}
 scapula_fr_l {parent=shoulder_fr_l, loc=<-5.0362 3.7723 -0.1058>, rot=<0.9487 -0.3158 0.0121 -0.0097>}
 shoulder_fr_r {parent=spine_03, loc=<1.9267 -0.6483 5.7706>, rot=<-0.0015 -0.0135 0.6066 -0.7949>}
 elbow_fr_r {parent=shoulder_fr_r, loc=<-10.8758 0 -0.0001>, rot=<-0.0045 0.004 0.1185 -0.9929>}
 wrist_fr_r {parent=elbow_fr_r, loc=<-11.4814 0 0>, rot=<0.9838 -0.1787 0.0143 -0.005>}
 ball_fr_r {parent=wrist_fr_r, loc=<-3.7813 0 0>, rot=<-0.0436 0.0565 -0.4345 -0.8978>}
 toe_fr_r {parent=ball_fr_r, loc=<-1.4235 0 0>, rot=<-0.0136 -0.0343 0.0437 -0.9984>}
 toe_tip_fr_r {parent=toe_fr_r, loc=<-3.1112 0 0.0001>, rot=<0.0031 0.006 -0.1897 -0.9818>}
 scapula_fr_r {parent=shoulder_fr_r, loc=<5.0362 -3.7723 0.1057>, rot=<0.9487 -0.3158 0.0121 -0.0097>}
 MSTR_wing_l {parent=spine_03, loc=<-6.1704 -8.179 -8.6482>, rot=<-0.199 0.0959 0.9752 -0.0123>}
 dyn_wing_l {parent=MSTR_wing_l, loc=<0.1 0 0>, rot=<0 0 0 1>}
 MSTR_wing_r {parent=spine_03, loc=<-6.1704 -8.179 8.6482>, rot=<-0.0959 -0.199 -0.0123 -0.9752>}
 dyn_wing_r {parent=MSTR_wing_r, loc=<-0.1 0 0>, rot=<0 0 0 1>}
 hips {parent=pelvis, loc=<0 0 0>, rot=<0 0 0 1>}
 thigh_bk_l {parent=hips, loc=<-1.7864 -5.2948 -5.6956>, rot=<0.7238 0.6899 -0.0073 0.0076>}
 knee_bk_l {parent=thigh_bk_l, loc=<12.276 0 0>, rot=<0.0064 -0.005 0.4145 -0.91>}
 ankle_bk_l {parent=knee_bk_l, loc=<11.9381 0 0>, rot=<-0.0046 0.0019 -0.5264 -0.8502>}
 ball_bk_l {parent=ankle_bk_l, loc=<10.3665 0 0>, rot=<-0.0049 -0.0045 -0.4831 -0.8756>}
 toe_bk_l {parent=ball_bk_l, loc=<1.8315 0 0>, rot=<0 0 0.0737 -0.9973>}
 toe_tip_bk_l {parent=toe_bk_l, loc=<2.4192 0 0>, rot=<0 0 -0.2485 -0.9686>}
 thigh_bk_r {parent=hips, loc=<-1.7864 -5.2948 5.6956>, rot=<0.6899 -0.7238 -0.0076 -0.0073>}
 knee_bk_r {parent=thigh_bk_r, loc=<-12.276 0 0>, rot=<0.0064 -0.005 0.4145 -0.91>}
 ankle_bk_r {parent=knee_bk_r, loc=<-11.9381 0 0>, rot=<-0.0046 0.0019 -0.5264 -0.8502>}
 ball_bk_r {parent=ankle_bk_r, loc=<-10.3664 0 0>, rot=<-0.0049 -0.0044 -0.4831 -0.8756>}
 toe_bk_r {parent=ball_bk_r, loc=<-1.8315 0 0>, rot=<0 0 0.0737 -0.9973>}
 toe_tip_bk_r {parent=toe_bk_r, loc=<-2.4192 0 0>, rot=<0 0 -0.2485 -0.9686>}
 BothLegIK {type=Curve, loc=<0 0 0>, rot=<0 0 0 1>}
 L_eye_look_up_pose {type=Curve, loc=<0 0 0>, rot=<0 0 0 1>}
 L_wide_pose {type=Curve, loc=<0 0 0>, rot=<0 0 0 1>}
 R_lip_corner_narrow_pose {type=Curve, loc=<0 0 0>, rot=<0 0 0 1>}
 R_upper_lip_lower_pose {type=Curve, loc=<0 0 0>, rot=<0 0 0 1>}
 L_lip_corner_narrow_pose {type=Curve, loc=<0 0 0>, rot=<0 0 0 1>}
 L_upper_lip_lower_pose {type=Curve, loc=<0 0 0>, rot=<0 0 0 1>}
 R_lower_lip_up_pose {type=Curve, loc=<0 0 0>, rot=<0 0 0 1>}
 R_brow_squeeze_in_pose {type=Curve, loc=<0 0 0>, rot=<0 0 0 1>}
 L_lip_corner_wide_pose {type=Curve, loc=<0 0 0>, rot=<0 0 0 1>}
 L_upperLidDown_pose {type=Curve, loc=<0 0 0>, rot=<0 0 0 1>}
 R_lower_lip_down_pose {type=Curve, loc=<0 0 0>, rot=<0 0 0 1>}
 R_upper_lip_raiser_pose {type=Curve, loc=<0 0 0>, rot=<0 0 0 1>}
 L_lower_lip_down_pose {type=Curve, loc=<0 0 0>, rot=<0 0 0 1>}
 L_upper_lip_raiser_pose {type=Curve, loc=<0 0 0>, rot=<0 0 0 1>}
 L_brow_squeeze_in_pose {type=Curve, loc=<0 0 0>, rot=<0 0 0 1>}
 L_lower_lip_up_pose {type=Curve, loc=<0 0 0>, rot=<0 0 0 1>}
 R_upperLidDown_pose {type=Curve, loc=<0 0 0>, rot=<0 0 0 1>}
 R_wide_pose {type=Curve, loc=<0 0 0>, rot=<0 0 0 1>}
 R_blink_pose {type=Curve, loc=<0 0 0>, rot=<0 0 0 1>}
 L_eye_look_left_pose {type=Curve, loc=<0 0 0>, rot=<0 0 0 1>}
 L_eye_look_down_pose {type=Curve, loc=<0 0 0>, rot=<0 0 0 1>}
 L_blink_pose {type=Curve, loc=<0 0 0>, rot=<0 0 0 1>}
 Reaction_Happy {type=Curve, loc=<0 0 0>, rot=<0 0 0 1>}
 L_eye_look_right_pose {type=Curve, loc=<0 0 0>, rot=<0 0 0 1>}
 R_lip_corner_wide_pose {type=Curve, loc=<0 0 0>, rot=<0 0 0 1>}
 L_smile_pose {type=Curve, loc=<0 0 0>, rot=<0 0 0 1>}
 R_smile_pose {type=Curve, loc=<0 0 0>, rot=<0 0 0 1>}
 L_frown_pose {type=Curve, loc=<0 0 0>, rot=<0 0 0 1>}
 R_frown_pose {type=Curve, loc=<0 0 0>, rot=<0 0 0 1>}
 C_nose_wrinkler_pose {type=Curve, loc=<0 0 0>, rot=<0 0 0 1>}
 L_cheek_up_pose {type=Curve, loc=<0 0 0>, rot=<0 0 0 1>}
 R_cheek_up_pose {type=Curve, loc=<0 0 0>, rot=<0 0 0 1>}
 jaw_open_pose {type=Curve, loc=<0 0 0>, rot=<0 0 0 1>}
 jaw_left_pose {type=Curve, loc=<0 0 0>, rot=<0 0 0 1>}
 jaw_right_pose {type=Curve, loc=<0 0 0>, rot=<0 0 0 1>}
 jaw_up_pose {type=Curve, loc=<0 0 0>, rot=<0 0 0 1>}
 jaw_forward_pose {type=Curve, loc=<0 0 0>, rot=<0 0 0 1>}
 jaw_back_pose {type=Curve, loc=<0 0 0>, rot=<0 0 0 1>}
 eye_convergence_pose {type=Curve, loc=<0 0 0>, rot=<0 0 0 1>}
 tongue_forward_pose {type=Curve, loc=<0 0 0>, rot=<0 0 0 1>}
 tongue_back_pose {type=Curve, loc=<0 0 0>, rot=<0 0 0 1>}
 tongue_up_pose {type=Curve, loc=<0 0 0>, rot=<0 0 0 1>}
 tongue_down_pose {type=Curve, loc=<0 0 0>, rot=<0 0 0 1>}
 L_brow_In_Up_pose {type=Curve, loc=<0 0 0>, rot=<0 0 0 1>}
 L_brow_Ot_Up_pose {type=Curve, loc=<0 0 0>, rot=<0 0 0 1>}
 L_brow_In_Dn_pose {type=Curve, loc=<0 0 0>, rot=<0 0 0 1>}
 L_brow_Ot_Dn_pose {type=Curve, loc=<0 0 0>, rot=<0 0 0 1>}
 R_brow_In_Up_pose {type=Curve, loc=<0 0 0>, rot=<0 0 0 1>}
 R_brow_Ot_Up_pose {type=Curve, loc=<0 0 0>, rot=<0 0 0 1>}
 R_brow_In_Dn_pose {type=Curve, loc=<0 0 0>, rot=<0 0 0 1>}
 R_brow_Ot_Dn_pose {type=Curve, loc=<0 0 0>, rot=<0 0 0 1>}
 upper_lip_up_pose {type=Curve, loc=<0 0 0>, rot=<0 0 0 1>}
 upper_lip_dn_pose {type=Curve, loc=<0 0 0>, rot=<0 0 0 1>}
 lower_lip_up_pose {type=Curve, loc=<0 0 0>, rot=<0 0 0 1>}
 lower_lip_dn_pose {type=Curve, loc=<0 0 0>, rot=<0 0 0 1>}
 L_smile_open_correction_pose {type=Curve, loc=<0 0 0>, rot=<0 0 0 1>}
 R_smile_open_correction_pose {type=Curve, loc=<0 0 0>, rot=<0 0 0 1>}
 tongue_left_pose {type=Curve, loc=<0 0 0>, rot=<0 0 0 1>}
 tongue_right_pose {type=Curve, loc=<0 0 0>, rot=<0 0 0 1>}
 reverse_jaw_pose {type=Curve, loc=<0 0 0>, rot=<0 0 0 1>}
 base_pose {type=Curve, loc=<0 0 0>, rot=<0 0 0 1>}
 R_eye_look_left_pose {type=Curve, loc=<0 0 0>, rot=<0 0 0 1>}
 R_eye_look_right_pose {type=Curve, loc=<0 0 0>, rot=<0 0 0 1>}
 R_eye_look_up_pose {type=Curve, loc=<0 0 0>, rot=<0 0 0 1>}
 R_eye_look_down_pose {type=Curve, loc=<0 0 0>, rot=<0 0 0 1>}
 L_lower_lid_up_pose {type=Curve, loc=<0 0 0>, rot=<0 0 0 1>}
 R_lower_lid_up_pose {type=Curve, loc=<0 0 0>, rot=<0 0 0 1>}
 tongue_curl_tip_up_pose {type=Curve, loc=<0 0 0>, rot=<0 0 0 1>}
 tongue_curl_tip_down_pose {type=Curve, loc=<0 0 0>, rot=<0 0 0 1>}
 tongue_curl_mid_up_pose {type=Curve, loc=<0 0 0>, rot=<0 0 0 1>}
 tongue_curl_mid_down_pose {type=Curve, loc=<0 0 0>, rot=<0 0 0 1>}
 L_tongue_twist_pose {type=Curve, loc=<0 0 0>, rot=<0 0 0 1>}
 R_tongue_twist_pose {type=Curve, loc=<0 0 0>, rot=<0 0 0 1>}
 AttachEmote1InPelvisSpace {type=Curve, loc=<0 0 0>, rot=<0 0 0 1>}
 AttachEmote2InPelvisSpace {type=Curve, loc=<0 0 0>, rot=<0 0 0 1>}
 AttachEmote3InPelvisSpace {type=Curve, loc=<0 0 0>, rot=<0 0 0 1>}
 AnimateAcc {type=Curve, loc=<0 0 0>, rot=<0 0 0 1>}
 pupil_shape {type=Curve, loc=<0 0 0>, rot=<0 0 0 1>}
 pupil_size {type=Curve, loc=<0 0 0>, rot=<0 0 0 1>}
 bypass_poseAsset {type=Curve, loc=<0 0 0>, rot=<0 0 0 1>}
]
