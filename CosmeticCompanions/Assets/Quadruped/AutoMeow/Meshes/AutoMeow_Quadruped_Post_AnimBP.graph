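# Post-process AnimBP for the AutoMeow quadruped cosmetic companion, adding procedural wing and collar motion to the input pose.
# Weighted spine/scapula constraints drive master wing bones, feeding two alternative spring AnimDynamics setups (world-space, bone-relative) joined by a bool blend.
# Translation constraints then pin dynamic wings to their masters, and collar bones follow neck, shoulder and scapula.
# FlourCut pose-asset nodes are unconnected; frontend flag FE updates without a visible reader.
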AnimBlueprint /CosmeticCompanions/Assets/Quadruped/AutoMeow/Meshes/AutoMeow_Quadruped_Post_AnimBP
target_skeleton=&/AutoMeow_Quadruped_Skeleton
graphs [
 AnimGraph {nodes [
   Pose!(
    PoseAsset:Pose Asset Object Reference=Get PoseAsset_0() -> (PoseAsset_0:Pose Asset Object Reference)
    SourcePose:Pose Link Structure=FlourCut_Quadruped_PoseAsset!(SourcePose:Pose Link Structure) -> (Pose:Pose Link Structure) {is=title AnimGraphNode_PoseBlendNode, poseAsset=/CosmeticCompanions/Assets/Quadruped/FlourCut/Meshes/FlourCut_Quadruped_PoseAsset}
   ) -> (Pose:Pose Link Structure) {is=title AnimGraphNode_PoseBlendNode, bIgnorePoseAssetSourceAnimationCurves=true, blendOption=Linear}
   n1 = "Constraint - dyn_wing_l (T)"(ComponentPose=Local To Component(LocalPose=Use cached pose 'PostDyn'()), Alpha=1, ConstraintWeights_0=1) {boneToModify {boneName=dyn_wing_l}, constraintSetup [{transformType=Translation, targetBone {boneName=MSTR_wing_l}}], constraintWeights [1]}
   n2 = "Constraint - dyn_wing_r (T)"(ComponentPose=@n1, Alpha=1, ConstraintWeights_0=1) {boneToModify {boneName=dyn_wing_r}, constraintSetup [{transformType=Translation, targetBone {boneName=MSTR_wing_r}}], constraintWeights [1]}
   n3 = "Constraint - extra_collar_l (PPPP)"!(
    ComponentPose=@n2
    Alpha=1
    ConstraintWeights_0=0.3
    ConstraintWeights_1:float=0.3
    ConstraintWeights_2:float=0.2
    ConstraintWeights_3:float=0.2
   ) -> (Pose) {
    boneToModify {boneName=extra_collar_l}
    constraintSetup [
     {targetBone {boneName=neck_01}}
     {targetBone {boneName=neck_02}}
     {targetBone {boneName=shoulder_fr_l}}
     {targetBone {boneName=scapula_fr_l}}
    ]
    constraintWeights [1,1,1,1]
   }
   n4 = "Constraint - extra_collar_r (PPPP)"!(
    ComponentPose=@n3
    Alpha=1
    ConstraintWeights_0=0.3
    ConstraintWeights_1:float=0.3
    ConstraintWeights_2:float=0.2
    ConstraintWeights_3:float=0.2
   ) -> (Pose) {
    boneToModify {boneName=extra_collar_r}
    constraintSetup [
     {targetBone {boneName=neck_01}}
     {targetBone {boneName=neck_02}}
     {targetBone {boneName=shoulder_fr_r}}
     {targetBone {boneName=scapula_fr_r}}
    ]
    constraintWeights [1,1,1,1]
   }
   Output Pose(Result=Component To Local(ComponentPose=@n4))
   n5 = "Constraint - MSTR_wing_l (PPP)"!(
    ComponentPose=Local To Component(LocalPose=Input Pose())
    Alpha=1
    ConstraintWeights_0=0.15
    ConstraintWeights_1:float=0.6
    ConstraintWeights_2:float=0.25
   ) -> (Pose) {
    boneToModify {boneName=MSTR_wing_l}
    constraintSetup [{targetBone {boneName=spine_01}}, {targetBone {boneName=spine_02}}, {targetBone {boneName=scapula_fr_l}}]
    constraintWeights [1,1,1]
   }
   n6 = "Constraint - MSTR_wing_r (PPP)"!(ComponentPose=@n5, Alpha=1, ConstraintWeights_0=0.15, ConstraintWeights_1:float=0.6, ConstraintWeights_2:float=0.25) -> (Pose) {
    boneToModify {boneName=MSTR_wing_r}
    constraintSetup [{targetBone {boneName=spine_01}}, {targetBone {boneName=spine_02}}, {targetBone {boneName=scapula_fr_r}}]
    constraintWeights [1,1,1]
   }
   Save cached pose 'PreDyn'(Pose=Component To Local(ComponentPose=@n6)) {cacheName=PreDyn}
   n7 = Anim Dynamics - Bone: dyn_wing_r(ComponentPose=Local To Component(LocalPose=Use cached pose 'PreDyn'()), Alpha=1, ExternalForce="0,0,0") {
    angularDampingOverride=0.9
    angularSpringConstant=1
    bAngularSpring=true
    bUsePlanarLimit=false
    numSolverIterationsPostUpdate=4
    simulationSpace=World
    alphaScaleBiasClamp {bias=1, scale=-1}
    boundBone {boneName=dyn_wing_r}
    gravityOverride {z=-980}
    physicsBodyDefinitions [{
      boundBone {boneName=dyn_wing_r}
      boxExtents=<10 10 10>
      constraintSetup {angularLimitsMax=<0 0 0>, angularLimitsMin=<0 0 0>, angularTarget=<0 0 0>}
      localJointOffset=<-2 0 0>
     }]
    relativeSpaceBone {boneName=MSTR_wing_r}
    simSpaceSettings {maxAngularAcceleration=0.01, maxAngularVelocity=0.01, simSpaceAngularAlpha=1}
   }
   n8 = Anim Dynamics - Bone: dyn_wing_l(ComponentPose=@n7, Alpha=1, ExternalForce="0,0,0") {
    angularDampingOverride=0.9
    angularSpringConstant=1
    bAngularSpring=true
    bUsePlanarLimit=false
    numSolverIterationsPostUpdate=4
    simulationSpace=World
    alphaScaleBiasClamp {bias=1, scale=-1}
    boundBone {boneName=dyn_wing_l}
    gravityOverride {z=-980}
    physicsBodyDefinitions [{
      boundBone {boneName=dyn_wing_l}
      boxExtents=<10 10 10>
      constraintSetup {angularLimitsMax=<0 0 0>, angularLimitsMin=<0 0 0>, angularTarget=<0 0 0>}
      localJointOffset=<-2 0 0>
     }]
    relativeSpaceBone {boneName=MSTR_wing_r}
    simSpaceSettings {maxAngularAcceleration=0.01, maxAngularVelocity=0.01, simSpaceAngularAlpha=1}
   }
   n9 = Anim Dynamics - Bone: dyn_wing_r(ComponentPose=Local To Component(LocalPose=Use cached pose 'PreDyn'()), Alpha=1, ExternalForce="0,0,0") {
    angularDampingOverride=1
    angularSpringConstant=10
    bAngularSpring=true
    bOverrideAngularDamping=true
    numSolverIterationsPostUpdate=4
    simulationSpace=BoneRelative
    alphaScaleBiasClamp {bias=1, scale=-1}
    boundBone {boneName=dyn_wing_r}
    componentAppliedLinearAccClamp=<1500 1500 1500>
    gravityOverride {z=-980}
    physicsBodyDefinitions [{
      boundBone {boneName=dyn_wing_r}
      boxExtents=<10 10 10>
      constraintSetup {angularLimitsMax=<15 15 15>, angularLimitsMin=<-15 -15 -15>, angularTarget=<1 0 0>}
      localJointOffset=<-8 0 0>
     }]
    relativeSpaceBone {boneName=MSTR_wing_r}
   }
   n10 = Anim Dynamics - Bone: dyn_wing_l(ComponentPose=@n9, Alpha=1, ExternalForce="0,0,0") {
    angularDampingOverride=1
    angularSpringConstant=10
    bAngularSpring=true
    bOverrideAngularDamping=true
    numSolverIterationsPostUpdate=4
    simulationSpace=BoneRelative
    alphaScaleBiasClamp {bias=1, scale=-1}
    boundBone {boneName=dyn_wing_l}
    componentAppliedLinearAccClamp=<1500 1500 1500>
    gravityOverride {z=-980}
    physicsBodyDefinitions [{
      boundBone {boneName=dyn_wing_l}
      boxExtents=<10 10 10>
      constraintSetup {angularLimitsMax=<15 15 15>, angularLimitsMin=<-15 -15 -15>, angularTarget=<1 0 0>}
      localJointOffset=<8 0 0>
     }]
    relativeSpaceBone {boneName=MSTR_wing_l}
   }
   Save cached pose 'PostDyn'(Pose=Blend Poses by bool(BlendPose_0=Component To Local(ComponentPose=@n8), BlendPose_1=Component To Local(ComponentPose=@n10))) {cacheName=PostDyn}
   Get PoseAsset() -> (PoseAsset:Pose Asset Object Reference)
  ]}
 EventGraph {nodes [
   Event BlueprintUpdateAnimation()
   > Set FE(FE:bool=IsInFrontend!() -> (ReturnValue:bool)) -> (Output_Get:bool)
  ]}
]
variables [
 FE {type=boolean}
 MeshComponent {type=SkeletalMeshComponent}
 NewVar {type=number}
 NewVar_0 {type=number}
 PoseAsset {type=PoseAsset, default=/CosmeticCompanions/Automation_Template/companion_pose_asset_template}
 PoseAsset_0 {type=PoseAsset, default=&FlourCut_Quadruped_PoseAsset}
]
